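# Translucent unlit material for the Tech Mage glider's uncloak effect, faking shading from its textures and a view-angle term.
# A directional wipe, broken up by hex-grid and scrolling tech-line textures projected on pre-skinned position, drives opacity and curve-atlas glow.
# Subtracted sphere masks add pulse and end-ring emissive.
# A custom-depth comparison masks occluded pixels, and a Debug switch, on by default, replaces the animation parameters with looping sines.
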Material /BRCosmetics/Gadgets/Assets/VinderTech_GliderChute/Glider_TechMage/FX/M_Glider_TechMage_Uncloak
graph {
 properties {
  allowTranslucentCustomDepthWrites=true
  bUsedWithMeshDeformer=true
  bUsedWithSkeletalMesh=true
  blendMode=BLEND_Translucent
  pixelDepthOffsetMode=PDOM_Legacy
  shadingModel=MSM_Unlit
  translucencyLightingMode=TLM_VolumetricPerVertexDirectional
  usedShadingModels=Unlit
 }
 nodes [
  n1 = TransformPosition(None=ObjectPositionWS()) {transformSourceType=TRANSFORMPOSSOURCE_World}
  n2 = Add(
   A=VertexInterpolator(VS=PreSkinnedPosition()).PS
   B=Transform(None=Constant3Vector() {constant=<rgba 0 -120 0 0>}) {transformSourceType=TRANSFORMSOURCE_Local, transformType=TRANSFORM_View}
  )
  n3 = Constant() {r=1000}
  n4 = Constant() {r=1.5}
  n5 = TextureSample(UVs=TextureCoordinate() {uTiling=1.5, vTiling=1.5}) {samplerType=SAMPLERTYPE_Grayscale, texture=&LowResBlurredNoise}
  n6 = LinearInterpolate(A=Divide(A=@n3, B=@n4), B=@n3, Alpha=@n5.R)
  n7 = StaticSwitchParameter(
   True=MaterialFunctionCall(Speed=Constant() {r=0.3}) {materialFunction=&DebugTimeSine}.Result
   False=ScalarParameter() {parameterName=BP_UnCloakingPulse}
  ) {defaultValue=true, parameterName=Debug}
  n8 = Constant() {r=1700}
  n9 = LinearInterpolate(A=LinearInterpolate(A=Divide(A=@n8, B=@n4), B=@n8, Alpha=@n5.R), Alpha=@n7) {constB=0}
  n10 = Multiply(A=@n9) {constB=5}
  n11 = Subtract(
   A=SphereMask(A=@n1, B=@n2, Radius=LinearInterpolate(A=@n6, Alpha=@n7) {constB=0}) {hardnessPercent=50}
   B=SphereMask(A=@n1, B=@n2, Radius=LinearInterpolate(A=@n10, B=@n9, Alpha=@n7)) {hardnessPercent=5}
  )
  n12 = VectorParameter() {parameterName=EmissiveColor, defaultValue=<rgba 0.2316 10 6.4321 1>}
  n13 = Multiply(A=Multiply(A=Power(Base=@n11)) {constB=12}, B=@n12.RGB)
  n14 = Add(
   A=VertexInterpolator(VS=PreSkinnedPosition()).PS
   B=Transform(None=Constant3Vector() {constant=<rgba 0 -80 0 0>}) {transformSourceType=TRANSFORMSOURCE_Local, transformType=TRANSFORM_View}
  )
  n15 = TransformPosition(None=ObjectPositionWS()) {transformSourceType=TRANSFORMPOSSOURCE_World}
  n16 = Constant() {r=1100}
  n17 = StaticSwitchParameter(
   True=MaterialFunctionCall(Speed=Constant() {r=0.3}) {materialFunction=&DebugTimeSine}.Result
   False=ScalarParameter() {parameterName=EndActive}
  ) {defaultValue=true, parameterName=Debug}
  n18 = Constant() {r=1000}
  n19 = Subtract(
   A=SphereMask(A=@n14, B=@n15, Radius=LinearInterpolate(B=@n16, Alpha=@n17)) {hardnessPercent=20}
   B=SphereMask(A=@n14, B=@n15, Radius=LinearInterpolate(B=@n18, Alpha=@n17)) {hardnessPercent=0}
  )
  n20 = Saturate(None=Subtract(A=Power(Base=@n19), B=Divide(A=@n17) {constB=30}))
  n21 = TextureObject() {texture=/Game/Effects/Fort_Effects/Textures/T_TechLines}
  n22 = ComponentMask(None=VectorParameter() {parameterName=TextureSize_TechLine_End, defaultValue=<rgba 30 95 30 1>}.RGB) {b=true, g=true, r=true}
  n23 = VertexInterpolator(VS=PreSkinnedPosition())
  n24 = ScalarParameter() {defaultValue=0.25, parameterName=RotZAxis}
  n25 = ScalarParameter() {parameterName=RotXAxis}
  n26 = Multiply(A=ComponentMask(None=MaterialFunctionCall(TextureObject=@n21, TextureSize=@n22, WorldPosition=@n23.PS, RotZAxis=@n24, RotXAxis=@n25) {materialFunction=/Game/Weapons/WeaponSkins/Wraps/MaterialFunctions/WorldAlignedTexture}.XYZ Texture) {r=true}) {constB=10}
  n27 = Multiply(A=Add(A=@n20, B=Multiply(A=@n20, B=@n26) {constB=0}), B=Divide(A=@n12.RGB) {constB=5})
  n28 = TextureSample() {texture=/BRCosmetics/Gadgets/Assets/VinderTech_GliderChute/Glider_TechMage/Textures/T_Glider_Tech_Mage_D}
  n29 = VectorParameter() {parameterName=BaseColorBias, defaultValue=<rgba 1 1 1 1>}
  n30 = TextureSample() {samplerType=SAMPLERTYPE_Normal, texture=/BRCosmetics/Gadgets/Assets/VinderTech_GliderChute/Glider_TechMage/Textures/T_Glider_Tech_Mage_N}
  n31 = TextureSample() {samplerType=SAMPLERTYPE_Masks, texture=/BRCosmetics/Gadgets/Assets/VinderTech_GliderChute/Glider_TechMage/Textures/T_Glider_Tech_Mage_S}
  n32 = Power(Base=@n31.B)
  n33 = Constant() {r=1}
  n34 = Transform(None=MaterialFunctionCall(
    BaseNormal=@n30.RGB
    AdditionalNormal=MaterialFunctionCall(R=@n32, G=@n32, B=@n33) {materialFunction=&AppendMany}.RGB
   ) {materialFunction=&BlendAngleCorrectedNormals}.Result)
  n35 = Constant() {r=1}
  n36 = ShadingPathSwitch(
   Default=MaterialFunctionCall("FadeSensitivity (optional)"=ScalarParameter() {defaultValue=5, parameterName=SpecRange}) {materialFunction=&FadeBasedOnViewAngle}.Result
   Forward=@n35
   Mobile=@n35
  ) {desc=Migrated from FeatureLevelSwitch}
  n37 = Saturate(None=Abs(None=DotProduct(A=@n34, B=@n36)))
  n38 = Constant() {r=0.1}
  n39 = TransformPosition(None=WorldPosition() {worldPositionShaderOffset=WPT_ExcludeAllShaderOffsets}.XYZ) {transformSourceType=TRANSFORMPOSSOURCE_World}
  n40 = TransformPosition(None=ObjectPositionWS()) {transformSourceType=TRANSFORMPOSSOURCE_World}
  n41 = VectorParameter() {parameterName=TextureSize_TechLine, defaultValue=<rgba 1 100 20 1>}
  n42 = Multiply(
   A=MaterialFunctionCall(
    TextureObject=TextureObject() {samplerType=SAMPLERTYPE_Masks, texture=/Game/Accessories/FORT_Backpacks/F_MED_CyberRunner_Backpack/Textures/HexGrid_M}
    TextureSize=ComponentMask(None=@n41.RGB) {b=true, g=true, r=true}
    WorldPosition=@n23.PS
    RotZAxis=ScalarParameter() {defaultValue=0.25, parameterName=RotZAxis}
    RotXAxis=ScalarParameter() {parameterName=RotXAxis}
   ) {materialFunction=&WorldAlignedTexture}.XYZ Texture
   B=ScalarParameter() {defaultValue=1, parameterName=TechLineOpacity}
  )
  n43 = Constant()
  n44 = ComponentMask(None=QualitySwitch(
    Default=ShadingPathSwitch(Default=@n42, Deferred=@n42, Mobile=@n43) {desc=Migrated from FeatureLevelSwitch}
    Low=@n43
    High=@n42
    Medium=@n42
   )) {r=true}
  n45 = VectorParameter() {parameterName=TextureSize_Triangle, defaultValue=<rgba 80 80 80 1>}
  n46 = Multiply(A=Time()) {constB=60}
  n47 = ComponentMask(None=MaterialFunctionCall(
    TextureObject=TextureObject() {texture=&T_TechLines}
    TextureSize=ComponentMask(None=@n45.RGB) {b=true, g=true, r=true}
    WorldPosition=Add(A=@n23.PS, B=@n46)
   ) {materialFunction=/Engine/Functions/Engine_MaterialFunctions01/Texturing/WorldAlignedTexture}.XYZ Texture) {b=true}
  n48 = Multiply(A=@n47) {constB=3}
  n49 = ScalarParameter() {defaultValue=-1, parameterName=UnCloakingDirection, sliderMax=1, sliderMin=-1}
  n50 = Multiply(
   A=AppendVector(A=ComponentMask(None=@n40) {g=true}, B=ComponentMask(None=@n40) {b=true, g=true})
   B=Multiply(A=Add(A=@n44, B=@n48), B=Multiply(A=@n49) {constB=-1}) {constB=-1}
  )
  n51 = Multiply(
   A=StaticComponentMaskParameter(None=Subtract(A=@n39, B=@n50)) {defaultG=true, parameterName=UnCloakChannel}
   B=ScalarParameter() {defaultValue=-1, parameterName=UnCloakingDirection, sliderMax=1, sliderMin=-1}
  ) {constB=-1}
  n52 = ScalarParameter() {defaultValue=20, parameterName=Soften}
  n53 = ScalarParameter() {defaultValue=-10, parameterName=LerpLow}
  n54 = ScalarParameter() {defaultValue=12, parameterName=LerpHigh}
  n55 = StaticSwitchParameter(
   True=MaterialFunctionCall(Speed=Constant() {r=0.3}) {materialFunction=&DebugTimeSine}.Result
   False=ScalarParameter() {parameterName=BP_UnCloaking}
  ) {defaultValue=true, parameterName=Debug}
  n56 = Saturate(None=Subtract(A=Divide(A=@n51, B=@n52) {constB=200}, B=LinearInterpolate(A=@n53, B=@n54, Alpha=@n55) {constA=-2.5}))
  n57 = Add(A=Multiply(A=@n56, B=@n44), B=Multiply(A=@n56, B=Multiply(A=@n47) {constB=0.05}))
  n58 = CurveAtlasRowParameter(CurveTime=@n57) {atlas=&AthenaCurveAtlas01, curve=/Game/Athena/LinearColorCurves/Curve_TechGreen, defaultValue=96}
  n59 = Power(Base=@n55) {constExponent=1}
  n60 = LinearInterpolate(
   B=Multiply(
    A=MaterialFunctionCall(Hue Shift Percentage=@n38, Texture=@n58) {materialFunction=&HueShift}.Result
    B=ScalarParameter() {defaultValue=1, group=UnCloaking, parameterName=EmissiveMultiple}
   )
   Alpha=@n59
  )
  n61 = TextureSample() {texture=/BRCosmetics/Gadgets/Assets/VinderTech_GliderChute/Glider_TechMage/Textures/T_Glider_Tech_Mage_E}
  n62 = ScalarParameter() {defaultValue=1, parameterName=DeployedGlow}
  n63 = Add(
   A=LinearInterpolate(A=Multiply(A=@n28.RGB, B=@n29.RGB), B=Multiply(A=@n28.RGB) {constB=5}, Alpha=Power(Base=@n37) {constExponent=1.5})
   B=Add(A=Multiply(A=@n60, B=@n57), B=Multiply(A=@n61.RGB, B=@n62)) {constB=0}
  )
  n64 = Multiply(A=@n63, B=@n12.RGB)
  MaterialFunctionCall(Emissive=Add(A=Add(A=@n13, B=@n27) {constB=0}, B=LinearInterpolate(A=@n63, B=@n64, Alpha=@n59))) {materialFunction=&MF_FX_EmissiveAdjustments}
  n65 = Constant() {r=1}
  n66 = If(
   A=Add(A=PixelDepth().R, B=ScalarParameter() {defaultValue=-0.7, parameterName=DepthBias})
   B=ComponentMask(None=SceneTexture() {sceneTextureId=PPI_CustomDepth}.Color) {r=true}
   "A > B"=Constant()
   "A == B"=@n65
   "A < B"=@n65
  )
  Multiply(
   A=Add(A=Multiply(A=@n56, B=@n44), B=@n56)
   B=ShadingPathSwitch(Default=@n66, Deferred=@n66, Mobile=Constant() {r=1}) {desc=Migrated from FeatureLevelSwitch}
  )
 ]
}
